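# Additive unlit Niagara sprite material, named for the swipe glow in the MetalScout pickaxe impact effect.
# A clamped flare texture scrolls along V via the WipePosition dynamic parameter, with its RGB added to particle color.
# Its red channel is shaped by a soft centered sphere mask and multiplied by particle alpha.
# A quality switch adds depth fade only on High and Epic.
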Material /BRCosmetics/Weapons/FORT_Melee/Pickaxe_MetalScout/FX/M_FX_Pickaxe_MetalScoutImpact_SwipeGlow
graph {
 properties {bUsedWithNiagaraSprites=true, blendMode=BLEND_Additive, pixelDepthOffsetMode=PDOM_Legacy, shadingModel=MSM_Unlit, usedShadingModels=Unlit}
 nodes [
  n1 = ParticleColor()
  n2 = TextureCoordinate()
  n3 = AppendVector(A=Constant(), B=DynamicParameter() {paramNames [WipePosition,Param2,Param3,Param4]}.WipePosition)
  n4 = TextureSampleParameter2D(UVs=Add(A=@n2, B=@n3)) {parameterName=T_FlareTex, samplerSource=SSM_Clamp_WorldGroupSettings, texture=/Game/Effects/Fort_Effects/Textures/T_SUN_Glare_A}
  Add(A=@n1.RGB, B=@n4.RGB)
  n5 = TextureCoordinate()
  n6 = Constant() {r=0.5}
  n7 = Multiply(A=@n4.R, B=Power(Base=SphereMask(A=@n5, B=@n6) {attenuationRadius=0.5, hardnessPercent=0}) {constExponent=1})
  n8 = DepthFade(Opacity=@n7, FadeDistance=ScalarParameter() {defaultValue=10, parameterName=DepthFade_Dist}) {fadeDistanceDefault=50}
  Multiply(A=@n1.A, B=QualitySwitch(Default=@n7, Low=@n7, High=@n8, Medium=@n7, Epic=@n8))
  Multiply()
  SphereMask(A=@n2, B=@n3) {attenuationRadius=0.5, hardnessPercent=0}
 ]
}
